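# Physics asset for the belt mesh of the female medium AutumnDelivery cosmetic body, defining a hanging belt piece and a sword.
# Two three-bone dyn_ chains descend from MSTR_ bones under the pelvis, with zero twist limits and tighter swing on the sword, especially its last joint.
# Capsules cover pelvis, thighs and calves, joined by free angular constraints.
# The last sword bone uses four capsules to approximate a larger shape.
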PhysicsAsset /BRCosmetics/Characters/Player/Female/Medium/Bodies/F_MED_AutumnDelivery/Meshes/F_MED_AutumnDelivery_Belt_Physics
bodies {
 MSTR_belt_hanging [{shapeType=Capsule, length=3, radius=3, shapeName=MSTR_belt_hanging_capsule, rotation=<rot -90 0 0>}]
 MSTR_sword [{shapeType=Capsule, length=0.505, radius=0.505, shapeName=MSTR_sword_capsule, rotation=<rot 0 -90 0>}]
 calf_l [{shapeType=Capsule, length=28.6893, radius=8, shapeName=calf_l_capsule, center=<-23.7889 0 1.5914>, rotation=<rot -85.2898 179.9953 179.9959>}]
 calf_r [{shapeType=Capsule, length=28.6893, radius=8, shapeName=calf_r_capsule, center=<23.6448 0 -2.128>, rotation=<rot -86.7531 180 179.9999>}]
 dyn_belt_hanging_1 [{shapeType=Capsule, length=11, radius=4, shapeName=dyn_belt_hanging_1_capsule, center=<10 -5 0>, rotation=<rot -90 180 -175>}]
 dyn_belt_hanging_2 [{shapeType=Capsule, length=11, radius=4, shapeName=dyn_belt_hanging_2_capsule, center=<10 -4 0>, rotation=<rot -90 180 -175>}]
 dyn_belt_hanging_3 [{shapeType=Capsule, length=8, radius=4, shapeName=dyn_belt_hanging_3_capsule, center=<10 -2.5 0>, rotation=<rot -90 -90 97>}]
 dyn_sword_1 [{
   shapeType=Capsule
   length=10.7283
   radius=5.2106
   shapeName=dyn_sword_1_capsule
   center=<9.9636 -4.0646 0>
   rotation=<rot -90 -162.8407 170.8226>
  }]
 dyn_sword_2 [{shapeType=Capsule, length=8, radius=4, shapeName=dyn_sword_2_capsule, center=<10 -2.8 0>, rotation=<rot -90 180 180>}]
 dyn_sword_3 [
  {shapeType=Capsule, length=8, radius=4, shapeName=dyn_sword_3_capsule, center=<10 -1.7 0>, rotation=<rot 0 180 180>}
  {
   shapeType=Capsule
   length=25.9359
   radius=3.8436
   shapeName=dyn_sword_3_capsule
   center=<-1.7091 -5.885 -11.5016>
   rotation=<rot 78.1996 0.0006 -4.4586>
  }
  {
   shapeType=Capsule
   length=15
   radius=3.804
   shapeName=dyn_sword_3_capsule
   center=<-4.5353 -6.3281 -23.1942>
   rotation=<rot -11.619 -0.0089 5.9345>
  }
  {
   shapeType=Capsule
   length=33.9682
   radius=7.0978
   shapeName=dyn_sword_3_capsule
   center=<2.7132 -6.3232 13.4779>
   rotation=<rot -11.3818 0 5.3218>
  }
 ]
 pelvis [{shapeType=Capsule, length=6.5187, radius=13.0473, shapeName=pelvis_capsule, center=<-8.0253 1.247 0>}]
 thigh_l [{shapeType=Capsule, length=31.3253, radius=12.328, shapeName=thigh_l_capsule, center=<-18.448 3.3413 1>, rotation=<rot -90 0 -8>}]
 thigh_r [{
   shapeType=Capsule
   length=36.385
   radius=9.1642
   shapeName=thigh_r_capsule
   center=<18.0916 0 -1.9932>
   rotation=<rot -87.5215 -0.0002 0.0002>
  }]
}
constraints [
 {
  bone1Name=calf_r
  bone2Name=thigh_r
  swing1LimitDegrees=45
  swing1Motion=Free
  swing2LimitDegrees=45
  swing2Motion=Free
  twistLimitDegrees=45
  twistMotion=Free
 }
 {
  bone1Name=calf_l
  bone2Name=thigh_l
  swing1LimitDegrees=45
  swing1Motion=Free
  swing2LimitDegrees=45
  swing2Motion=Free
  twistLimitDegrees=45
  twistMotion=Free
 }
 {
  bone1Name=thigh_l
  bone2Name=pelvis
  swing1LimitDegrees=45
  swing1Motion=Free
  swing2LimitDegrees=45
  swing2Motion=Free
  twistLimitDegrees=45
  twistMotion=Free
 }
 {
  bone1Name=thigh_r
  bone2Name=pelvis
  swing1LimitDegrees=45
  swing1Motion=Free
  swing2LimitDegrees=45
  swing2Motion=Free
  twistLimitDegrees=45
  twistMotion=Free
 }
 {
  bone1Name=MSTR_belt_hanging
  bone2Name=pelvis
  swing1LimitDegrees=45
  swing1Motion=Free
  swing2LimitDegrees=45
  swing2Motion=Free
  twistLimitDegrees=45
  twistMotion=Free
 }
 {
  bone1Name=MSTR_sword
  bone2Name=pelvis
  swing1LimitDegrees=45
  swing1Motion=Free
  swing2LimitDegrees=45
  swing2Motion=Free
  twistLimitDegrees=45
  twistMotion=Free
 }
 {bone1Name=dyn_sword_1, bone2Name=MSTR_sword, swing1LimitDegrees=25, swing2LimitDegrees=25, twistLimitDegrees=0}
 {bone1Name=dyn_sword_2, bone2Name=dyn_sword_1, swing1LimitDegrees=25, swing2LimitDegrees=25, twistLimitDegrees=0}
 {bone1Name=dyn_sword_3, bone2Name=dyn_sword_2, swing1LimitDegrees=5, swing2LimitDegrees=5, twistLimitDegrees=0}
 {bone1Name=dyn_belt_hanging_1, bone2Name=MSTR_belt_hanging, swing1LimitDegrees=50, swing2LimitDegrees=50, twistLimitDegrees=0}
 {bone1Name=dyn_belt_hanging_2, bone2Name=dyn_belt_hanging_1, swing1LimitDegrees=50, swing2LimitDegrees=50, twistLimitDegrees=0}
 {bone1Name=dyn_belt_hanging_3, bone2Name=dyn_belt_hanging_2, swing1LimitDegrees=50, swing2LimitDegrees=50, twistLimitDegrees=0}
]
