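# Physics asset for the left boot of the female medium AutumnDelivery body cosmetic, covering the calf and a dynamic boot belt.
# Four capsule bodies form a chain from the calf through a master belt bone to two smaller belt segments.
# The master-to-calf joint swings and twists freely, while the two segment joints have limited swing and a zero-degree twist limit.
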PhysicsAsset /BRCosmetics/Characters/Player/Female/Medium/Bodies/F_MED_AutumnDelivery/Meshes/F_MED_AutumnDelivery_BootL_Physics
bodies {
 MSTR_belt_boot_l [{shapeType=Capsule, length=10, radius=8, shapeName=MSTR_belt_boot_l_capsule, rotation=<rot -90 0 0>}]
 calf_l [{shapeType=Capsule, length=28.6893, radius=8, shapeName=calf_l_capsule, center=<-23.7889 0 1.5914>, rotation=<rot -85.2898 179.9953 179.9959>}]
 dyn_belt_boot_l_1 [{shapeType=Capsule, length=6, radius=4, shapeName=dyn_belt_boot_l_1_capsule, center=<9.8923 -4.8 0>, rotation=<rot -90 0 -11.3438>}]
 dyn_belt_boot_l_2 [{shapeType=Capsule, length=6, radius=4, shapeName=dyn_belt_boot_l_2_capsule, center=<10 -5 0>, rotation=<rot -90 0 0>}]
}
constraints [
 {bone1Name=dyn_belt_boot_l_1, bone2Name=MSTR_belt_boot_l, swing1LimitDegrees=30, swing2LimitDegrees=25, twistLimitDegrees=0}
 {bone1Name=dyn_belt_boot_l_2, bone2Name=dyn_belt_boot_l_1, swing1LimitDegrees=30, swing2LimitDegrees=25, twistLimitDegrees=0}
 {
  bone1Name=MSTR_belt_boot_l
  bone2Name=calf_l
  swing1LimitDegrees=45
  swing1Motion=Free
  swing2LimitDegrees=45
  swing2Motion=Free
  twistLimitDegrees=45
  twistMotion=Free
 }
]
