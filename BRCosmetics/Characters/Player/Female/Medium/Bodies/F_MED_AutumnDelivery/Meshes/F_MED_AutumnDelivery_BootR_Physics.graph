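# Physics asset for the right boot mesh of the female medium AutumnDelivery body, apparently for secondary motion of a dangling boot part.
# Capsule bodies cover the calf, foot, boot master bone and two chained dynamic boot bones.
# The dynamic chain hangs from the master bone with limited swing and zero twist range.
# Foot and master bone both connect to the calf with fully free angular motion.
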PhysicsAsset /BRCosmetics/Characters/Player/Female/Medium/Bodies/F_MED_AutumnDelivery/Meshes/F_MED_AutumnDelivery_BootR_Physics
bodies {
 MSTR_bert_boot_r [{shapeType=Capsule, length=10, radius=8, shapeName=MSTR_bert_boot_r_capsule, rotation=<rot -90 0 0>}]
 calf_r [{shapeType=Capsule, length=28.6893, radius=8, shapeName=calf_r_capsule, center=<23.6448 0 -2.128>, rotation=<rot -86.7531 180 179.9999>}]
 dyn_bert_boot_r_1 [{shapeType=Capsule, length=4, radius=3, shapeName=dyn_bert_boot_r_1_capsule, center=<7.9223 -3.5 0>, rotation=<rot -90 180 167.589>}]
 dyn_bert_boot_r_2 [{shapeType=Capsule, length=4, radius=3, shapeName=dyn_bert_boot_r_2_capsule, center=<8 -4.2 0>, rotation=<rot -90 0 0>}]
 foot_r [{shapeType=Capsule, length=13.2817, radius=8, shapeName=foot_r_capsule, center=<4.4509 -8.593 -1.6681>, rotation=<rot 0 -0.0003 -86.5547>}]
}
constraints [
 {bone1Name=dyn_bert_boot_r_1, bone2Name=MSTR_bert_boot_r, swing1LimitDegrees=35, swing2LimitDegrees=25, twistLimitDegrees=0}
 {bone1Name=dyn_bert_boot_r_2, bone2Name=dyn_bert_boot_r_1, swing1LimitDegrees=35, swing2LimitDegrees=25, twistLimitDegrees=0}
 {
  bone1Name=MSTR_bert_boot_r
  bone2Name=calf_r
  swing1LimitDegrees=45
  swing1Motion=Free
  swing2LimitDegrees=45
  swing2Motion=Free
  twistLimitDegrees=45
  twistMotion=Free
 }
 {
  bone1Name=foot_r
  bone2Name=calf_r
  swing1LimitDegrees=45
  swing1Motion=Free
  swing2LimitDegrees=45
  swing2Motion=Free
  twistLimitDegrees=45
  twistMotion=Free
 }
]
