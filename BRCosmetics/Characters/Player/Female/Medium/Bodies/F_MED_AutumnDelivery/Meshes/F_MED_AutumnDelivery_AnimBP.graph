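# Post-copy animation blueprint for the AutumnDelivery female medium cosmetic body, adding secondary motion to the parent mesh's pose.
# A deform Control Rig feeds belt RigidBody physics, using pelvis space plus a cap tilt in the front end and root space otherwise.
# Sliding offsets the sword and hanging belt, then RigidBody boots and AnimDynamics bracelets and cap wing follow.
# A deprecated helper updates the gravity override for the RigidBody nodes each frame.
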AnimBlueprint /BRCosmetics/Characters/Player/Female/Medium/Bodies/F_MED_AutumnDelivery/Meshes/F_MED_AutumnDelivery_AnimBP
target_skeleton=&/F_MED_AutumnDelivery_Skeleton
graphs [
 AnimGraph {nodes [
   n1 = "Transform (Modify) Bone - Bone: MSTR_sword"(
    ComponentPose=Local To Component(LocalPose=Use cached pose 'CRCache'())
    Alpha=1
    Translation="-5,0,5"
    Rotation="-30,-100,-30"
    Scale="1,1,1"
   ) {
    rotationMode=BMM_Additive
    rotationSpace=BCS_BoneSpace
    translationMode=BMM_Additive
    translationSpace=BCS_BoneSpace
    alphaBoolBlend {blendOption=Linear}
    boneToModify {boneName=MSTR_sword}
    rotation=<rot -30 -100 -30>
    translation {x=-5, z=5}
   }
   n2 = "Transform (Modify) Bone - Bone: MSTR_belt_hanging"(ComponentPose=@n1, Alpha=1, Translation="0,0,0", Rotation="0,-60,0", Scale="1,1,1") {
    rotationMode=BMM_Additive
    rotationSpace=BCS_BoneSpace
    translationMode=BMM_Additive
    translationSpace=BCS_BoneSpace
    alphaBoolBlend {blendOption=Linear}
    boneToModify {boneName=MSTR_belt_hanging}
    rotation {yaw=-60}
   }
   n3 = Component To Local(ComponentPose=@n2)
   n4 = Use cached pose 'CRCache'()
   n5 = Get bIsSliding()
   n6 = Get GravityOverride()
   n7 = RigidBody!(
    ComponentPose=Local To Component(LocalPose=Blend Poses by bool(BlendPose_0=@n3, BlendPose_1=@n4, bActiveValue=@n5) {blendType=Linear})
    Alpha
    OverrideWorldGravity:Vector=@n6
    ExternalForce
   ) -> (Pose) {
    bClampLinearTranslationLimitToRefPose=true
    bOverrideWorldGravity=true
    overridePhysicsAsset=&/F_MED_AutumnDelivery_BootL_Physics
    simulationSpace=BaseBoneSpace
    alphaBoolBlend {blendOption=Linear}
    baseBoneRef {boneName=calf_l}
    componentAppliedLinearAccClamp=<3000 3000 3000>
    componentLinearAccScale=<0.4 0.3 0.3>
    componentLinearVelScale=<0.4 0.3 0.3>
    overrideWorldGravity {z=-980}
   }
   n8 = RigidBody(ComponentPose=@n7, OverrideWorldGravity=@n6) {
    bClampLinearTranslationLimitToRefPose=true
    bOverrideWorldGravity=true
    overridePhysicsAsset=&/F_MED_AutumnDelivery_BootR_Physics
    simulationSpace=BaseBoneSpace
    alphaBoolBlend {blendOption=Linear}
    baseBoneRef {boneName=calf_r}
    componentAppliedLinearAccClamp=<3000 3000 3000>
    componentLinearAccScale=<0.4 0.3 0.3>
    componentLinearVelScale=<0.4 0.3 0.3>
    overrideWorldGravity {z=-980}
   }
   n9 = "Transform (Modify) Bone - Bone: MSTR_bert_boot_r"(ComponentPose=@n8, Alpha=1, Translation="0,0,0", Rotation="5,0,0", Scale="1,1,1") {
    rotationMode=BMM_Additive
    translationSpace=BCS_BoneSpace
    alphaBoolBlend {blendOption=Linear}
    boneToModify {boneName=MSTR_bert_boot_r}
    rotation {pitch=5}
   }
   n10 = Anim Dynamics - Bone: AD_braceletA(ComponentPose=@n9, Alpha=1, ExternalForce="0,0,0") {
    angularDampingOverride=1
    angularSpringConstant=20000
    bAngularSpring=true
    bOverrideAngularDamping=true
    bUseGravityOverride=true
    gravityScale=0.35
    simulationSpace=RootRelative
    alphaBoolBlend {blendOption=Linear}
    boundBone {boneName=AD_braceletA}
    componentLinearAccScale=<1 1 1>
    componentLinearVelScale=<1 1 1>
    gravityOverride {z=-980}
    physicsBodyDefinitions [{
      boundBone {boneName=AD_braceletA}
      boxExtents=<20 30 30>
      constraintSetup {angularLimitsMax=<0 5 5>, angularLimitsMin=<0 -5 -5>, angularTarget=<1 0 0>}
      localJointOffset=<4 0 0>
     }]
   }
   n11 = Anim Dynamics - Bone: AD_braceletB(ComponentPose=@n10, Alpha=1, ExternalForce="0,0,0") {
    angularDampingOverride=1
    angularSpringConstant=20000
    bAngularSpring=true
    bOverrideAngularDamping=true
    bUseGravityOverride=true
    gravityScale=0.35
    simulationSpace=RootRelative
    alphaBoolBlend {blendOption=Linear}
    boundBone {boneName=AD_braceletB}
    componentLinearAccScale=<1 1 1>
    componentLinearVelScale=<1 1 1>
    gravityOverride {z=-980}
    physicsBodyDefinitions [{
      boundBone {boneName=AD_braceletB}
      boxExtents=<20 30 30>
      constraintSetup {angularLimitsMax=<0 5 5>, angularLimitsMin=<0 -5 -5>, angularTarget=<1 0 0>}
      localJointOffset=<3 0 0>
     }]
   }
   n12 = Anim Dynamics - Bone: AD_braceletC(ComponentPose=@n11, Alpha=1, ExternalForce="0,0,0") {
    angularDampingOverride=1
    angularSpringConstant=20000
    bAngularSpring=true
    bOverrideAngularDamping=true
    bUseGravityOverride=true
    gravityScale=0.35
    simulationSpace=RootRelative
    alphaBoolBlend {blendOption=Linear}
    boundBone {boneName=AD_braceletC}
    componentLinearAccScale=<1 1 1>
    componentLinearVelScale=<1 1 1>
    gravityOverride {z=-980}
    physicsBodyDefinitions [{
      boundBone {boneName=AD_braceletC}
      boxExtents=<20 30 30>
      constraintSetup {angularLimitsMax=<0 5 5>, angularLimitsMin=<0 -5 -5>, angularTarget=<1 0 0>}
      localJointOffset=<2 0 0>
     }]
   }
   n13 = Anim Dynamics - Bone: CR_cap_wing(ComponentPose=@n12, Alpha=1, ExternalForce="0,0,0") {
    angularDampingOverride=1
    angularSpringConstant=20000
    bAngularSpring=true
    bOverrideAngularDamping=true
    bUseGravityOverride=true
    gravityScale=0.35
    simulationSpace=RootRelative
    alphaBoolBlend {blendOption=Linear}
    boundBone {boneName=CR_cap_wing}
    componentLinearAccScale=<1 1 1>
    componentLinearVelScale=<1 1 1>
    gravityOverride {z=-980}
    physicsBodyDefinitions [{
      boundBone {boneName=CR_cap_wing}
      boxExtents=<30 40 40>
      constraintSetup {angularLimitsMax=<0 5 5>, angularLimitsMin=<0 -5 -5>, angularTarget=<1 0 0>}
      localJointOffset=<1.5 0 0>
     }]
   }
   Output Pose(Result=Component To Local(ComponentPose=@n13)) {layerGroup=None}
   n14 = RigidBody(ComponentPose=Local To Component(LocalPose=Use cached pose 'Skirt Cache'()), OverrideWorldGravity=@n6) {
    bClampLinearTranslationLimitToRefPose=true
    bOverrideWorldGravity=true
    overridePhysicsAsset=&/F_MED_AutumnDelivery_Belt_Physics
    simulationSpace=BaseBoneSpace
    alphaBoolBlend {blendOption=Linear}
    baseBoneRef {boneName=root}
    componentAppliedLinearAccClamp=<5000 5000 5000>
    componentLinearAccScale=<0.4 0.3 0.3>
    componentLinearVelScale=<0.4 0.3 0.3>
    overrideWorldGravity {z=-980}
   }
   n15 = RigidBody(ComponentPose=Local To Component(LocalPose=Use cached pose 'Skirt Cache'()), OverrideWorldGravity=@n6) {
    bClampLinearTranslationLimitToRefPose=true
    bOverrideWorldGravity=true
    overridePhysicsAsset=&F_MED_AutumnDelivery_Belt_Physics
    simulationSpace=BaseBoneSpace
    alphaBoolBlend {blendOption=Linear}
    baseBoneRef {boneName=pelvis}
    componentAppliedLinearAccClamp=<3000 3000 3000>
    componentLinearAccScale=<0.4 0.3 0.3>
    componentLinearVelScale=<0.4 0.3 0.3>
    overrideWorldGravity {z=-980}
   }
   n16 = "Transform (Modify) Bone - Bone: CR_cap"(ComponentPose=@n15, Alpha=1, Translation="0,0,-1", Rotation="0,0,-15", Scale="1,1,1") {
    rotationMode=BMM_Additive
    rotationSpace=BCS_BoneSpace
    translationMode=BMM_Additive
    translationSpace=BCS_BoneSpace
    alphaBoolBlend {blendOption=Linear}
    boneToModify {boneName=CR_cap}
    rotation {roll=-15}
    translation {z=-1}
   }
   Save cached pose 'CRCache'(Pose=Blend Poses by bool(BlendPose_0=Component To Local(ComponentPose=@n16), BlendPose_1=Component To Local(ComponentPose=@n14), bActiveValue=Get bIsInFrontEnd()) {blendType=Linear}) {cacheName=CRCache}
   n17 = "Transform (Modify) Bone - Bone: attach_backpack"(
    ComponentPose=Local To Component(LocalPose=Copy Pose From Mesh() {bUseAttachedParent=true})
    Alpha=1
    Translation="0,3,0"
    Rotation="0,0,0"
    Scale="1,1,1"
   ) {
    translationMode=BMM_Additive
    translationSpace=BCS_BoneSpace
    alphaBoolBlend {blendOption=Linear}
    boneToModify {boneName=attach_backpack}
    translation {y=3}
   }
   n18 = Control Rig(Source=Component To Local(ComponentPose=@n17)) {
    alphaBoolBlend {blendOption=Linear}
    controlRigAssetReference {blueprintRigClass=&/F_MED_DeformRig_AutumnDelivery_ControlRig.F_MED_DeformRig_AutumnDelivery_ControlRig_C}
   }
   Save cached pose 'Skirt Cache'(Pose=@n18) {cacheName=Skirt Cache}
  ]}
 EventGraph {nodes [
   Event BlueprintUpdateAnimation()
   > n1 = Gravity Override State Function DEPRECATED(CharacterPartAnimInstance=Self-Reference(), GameTimeInSeconds=GetGameTimeInSeconds(), GravityOverrideParams=Get Gravity Override Params())
   > Set GravityOverride(GravityOverride=@n1)
  ]}
]
variables [
 Gravity Override Params {type=GravityOverrideParamsStruct}
 GravityOverride {type=Vector}
]
