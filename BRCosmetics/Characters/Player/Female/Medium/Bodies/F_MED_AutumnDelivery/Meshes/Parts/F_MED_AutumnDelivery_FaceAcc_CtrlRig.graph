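# Control Rig for the face accessory part of the female medium AutumnDelivery body, which applies pose-driven corrective offsets to hair master bones.
# A spherical pose reader on upperarm_r weights a Modify Transforms that adjusts MSTR_hair_r, MSTR_hair_ba and possibly others, run from the sequence's A output.
# A second reader on head, with debug draw enabled, weights a rotation of MSTR_hair_l, but this chain has no execution input connected.
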ControlRigBlueprint /BRCosmetics/Characters/Player/Female/Medium/Bodies/F_MED_AutumnDelivery/Meshes/Parts/F_MED_AutumnDelivery_FaceAcc_CtrlRig
graphs [
 RigVMFunctionLibrary {links [], nodes []}
 RigVMModel {nodes [
   RigUnit_BeginExecution = &RigUnit_BeginExecution()
   RigVMFunction_Sequence = &RigVMFunction_Sequence(ExecuteContext=@RigUnit_BeginExecution.ExecutePin)
   n1 = &RigUnit_SphericalPoseReader(
    ActiveRegionScaleFactors=(PositiveWidth=1,NegativeWidth=1,PositiveHeight=1,NegativeHeight=1)
    ActiveRegionSize=0.1
    Debug=(bDrawDebug=false,bDraw2D=False,bDrawLocalAxes=False,DebugScale=25,DebugSegments=20,DebugThickness=0.25)
    DriverAxis=(X=-1,Y=0,Z=0)
    DriverItem=(Type=Bone,Name="upperarm_r")
    FalloffRegionScaleFactors=(PositiveWidth=1,NegativeWidth=1,PositiveHeight=1,NegativeHeight=1)
    FalloffSize=0.25
    FlipHeightScaling=False
    FlipWidthScaling=False
    OptionalParentItem=(Type=Bone,Name="None")
    RotationOffset=(X=-100,Y=0,Z=0)
    ExecutePin=@RigVMFunction_Sequence.A
   )
   > &RigUnit_ModifyTransforms(
    ItemToModify="((Item=(Type=Bone,Name=\"MSTR_hair_r\"),Transform=(Rotation=(X=0.0469,Y=-0.3185,Z=-0.0352,W=0.9461),Translation=(X=0,Y=0,Z=0),Scale3D=(X=1,Y=1,Z=1))),(Item=(Type=Bone,Name=\"MSTR_hair_ba\"),Transform=(Rotation=(X=0,Y=0,Z=0,W=1),Translation=(X=0,Y=0,Z=0),Scale3 ...(truncated)"
    WeightMinimum=0
    Weight=@n1.OutputParam
   )
   n2 = &RigUnit_SphericalPoseReader(
    ActiveRegionScaleFactors=(PositiveWidth=1,NegativeWidth=1,PositiveHeight=1,NegativeHeight=1)
    ActiveRegionSize=0.1
    Debug=(bDrawDebug=True,bDraw2D=False,bDrawLocalAxes=False,DebugScale=25,DebugSegments=20,DebugThickness=0.25)
    DriverAxis=(X=0,Y=-1,Z=0)
    DriverItem=(Type=Bone,Name="head")
    ExecutePin=()
    FalloffRegionScaleFactors=(PositiveWidth=1,NegativeWidth=1,PositiveHeight=1,NegativeHeight=1)
    FalloffSize=0.2
    FlipHeightScaling=False
    FlipWidthScaling=False
    OptionalParentItem=(Type=Bone,Name="None")
    RotationOffset=(X=-90,Y=0,Z=-20)
   )
   > &RigUnit_ModifyTransforms(
    ItemToModify=((Item=(Type=Bone,Name="MSTR_hair_l"),Transform=(Rotation=(X=0.0909,Y=-0.0272,Z=-0.2014,W=0.9749),Translation=(X=0,Y=0,Z=0),Scale3D=(X=1,Y=1,Z=1))))
    WeightMinimum=0
    Weight=@n2.OutputParam
   )
  ]}
]
hierarchy [
 root {loc=<0 0 0>, rot=<0 0 0 1>}
 pelvis {parent=root, loc=<0 -0.5301 98.0571>, rot=<0.0127 0.707 -0.0127 -0.707>}
 spine_01 {parent=pelvis, loc=<6.1208 -0.9328 0>, rot=<0 0 0.034 -0.9994>}
 spine_02 {parent=spine_01, loc=<6.5996 -0.2095 0>, rot=<0 0 0.0238 -0.9997>}
 spine_03 {parent=spine_02, loc=<6.4914 -0.1237 0>, rot=<0 0 0.0149 -0.9999>}
 spine_04 {parent=spine_03, loc=<6.9598 0.0794 0>, rot=<0 0 -0.0132 -0.9999>}
 spine_05 {parent=spine_04, loc=<6.9987 0 0>, rot=<0 0 -0.0015 -1>}
 neck_01 {parent=spine_05, loc=<11.3138 0.1529 0>, rot=<0 0 -0.2003 -0.9797>}
 neck_02 {parent=neck_01, loc=<4.5627 0 0>, rot=<0 0 0.0504 -0.9987>}
 head {parent=neck_02, loc=<5.3914 0 0>, rot=<0 0 0.115 -0.9934>}
 CR_earring1 {parent=head, loc=<-0.8927 0.4384 6.437>, rot=<0.0014 -0.342 0.9397 -0.0039>}
 AD_bangB_l {parent=head, loc=<6.8882 6.3618 -6.1969>, rot=<0.5689 -0.7075 -0.3088 -0.2836>}
 AD_bangA_l {parent=head, loc=<6.5151 9.2435 -4.2156>, rot=<0.4717 -0.7526 -0.3818 -0.2557>}
 AD_bangB_r {parent=head, loc=<7.3851 8.2346 5.4481>, rot=<0.4443 -0.7774 0.3922 0.2106>}
 AD_bangA_r {parent=head, loc=<6.2744 10.2732 3.323>, rot=<0.4137 -0.823 0.3648 0.1359>}
 MSTR_hair_ba {parent=head, loc=<11.6548 -6.3631 0>, rot=<-0.0216 -0.9998 0 0>}
 dyn_hair_ba_1 {parent=MSTR_hair_ba, loc=<0 0 0>, rot=<0 0 0 1>}
 dyn_hair_ba_2 {parent=dyn_hair_ba_1, loc=<8.3775 -2.9943 0>, rot=<0 0 0 1>}
 MSTR_hair_l {parent=head, loc=<9.4063 -1.0953 -7.7345>, rot=<0.0041 -1 0 0>}
 dyn_hair_l_1 {parent=MSTR_hair_l, loc=<0 0 0>, rot=<0 0 0 1>}
 dyn_hair_l_2 {parent=dyn_hair_l_1, loc=<8.1251 0 0>, rot=<0 0 0 1>}
 MSTR_hair_r {parent=head, loc=<9.4063 -1.0953 7.6844>, rot=<0.0041 -1 0 0>}
 dyn_hair_r_1 {parent=MSTR_hair_r, loc=<0 0 0>, rot=<0 0 0 1>}
 dyn_hair_r_2 {parent=dyn_hair_r_1, loc=<8.1251 0 0>, rot=<0 0 0 1>}
 MSTR_hairStrand_l {parent=head, loc=<3.4212 2.8931 -7.3995>, rot=<0.192 0.5801 0.7873 0.0818>}
 dyn_hairStrand_l_1 {parent=MSTR_hairStrand_l, loc=<0 0 0>, rot=<0 0 0 1>}
 dyn_hairStrand_l_2 {parent=dyn_hairStrand_l_1, loc=<5.0009 0 0>, rot=<0 0 0 1>}
 MSTR_earring {parent=head, loc=<0.445 0.4494 6.437>, rot=<0.0014 -0.342 0.9397 -0.0039>}
 dyn_earring_01 {parent=MSTR_earring, loc=<0 0 0>, rot=<0 0 0 1>}
 dyn_earring_02 {parent=dyn_earring_01, loc=<1.3377 0 0>, rot=<0 0 0 1>}
 dyn_earring_03 {parent=dyn_earring_02, loc=<1.1648 0 0>, rot=<0 0 0 1>}
 AD_bang_c {parent=head, loc=<6.5217 10.2531 0.6361>, rot=<0 0 1 -0.0041>}
 MSTR_earring_r {parent=head, loc=<0.4451 0.4495 -6.437>, rot=<-0.0017 0.4226 0.9063 -0.0037>}
 dyn_earring_01_r {parent=MSTR_earring_r, loc=<0 0 0>, rot=<0 0 0 1>}
 dyn_earring_02_r {parent=dyn_earring_01_r, loc=<1.338 0 0>, rot=<0 0 0 1>}
 dyn_earring_03_r {parent=dyn_earring_02_r, loc=<1.165 0 0>, rot=<0 0 0 1>}
 clavicle_r {parent=spine_05, loc=<5.4831 -0.1053 3.2946>, rot=<0.7289 0.0465 -0.6829 0.009>}
 upperarm_r {parent=clavicle_r, loc=<-9.8658 1.2136 0.4105>, rot=<0.0164 -0.3245 -0.0269 -0.9454>}
]
