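# Physics asset for the AutumnDelivery medium female body's face-accessory mesh, covering dangling earrings and hair.
# Capsule bodies form two three-link earring chains and two-link chains for back, left and right hair plus a left strand.
# Each chain hangs from an MSTR root body constrained to the head with free rotation, and the head links likewise to spine_05.
# The dyn_ links use narrow swing limits and zero-degree twist limits.
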PhysicsAsset /BRCosmetics/Characters/Player/Female/Medium/Bodies/F_MED_AutumnDelivery/Meshes/Parts/F_MED_AutumnDelivery_FaceAcc_Physics
bodies {
 MSTR_earring [{shapeType=Capsule, length=0.505, radius=0.505, shapeName=MSTR_earring_capsule, rotation=<rot 0 -90 0>}]
 MSTR_earring_r [{shapeType=Capsule, length=0.505, radius=0.505, shapeName=MSTR_earring_capsule, rotation=<rot 0 -90 0>}]
 MSTR_hairStrand_l [{shapeType=Capsule, length=0.505, radius=0.505, shapeName=MSTR_hairStrand_l_capsule, rotation=<rot 0 -90 0>}]
 MSTR_hair_ba [{shapeType=Capsule, length=0.505, radius=0.505, shapeName=MSTR_hair_ba_capsule, rotation=<rot 0 -90 0>}]
 MSTR_hair_l [{shapeType=Capsule, length=0.505, radius=0.505, shapeName=MSTR_hair_l_capsule, rotation=<rot 0 -90 0>}]
 MSTR_hair_r [{shapeType=Capsule, length=0.505, radius=0.505, shapeName=MSTR_hair_r_capsule, rotation=<rot 0 -90 0>}]
 dyn_earring_01 [{shapeType=Capsule, length=0.5, radius=0.5, shapeName=dyn_earring_01_capsule, center=<0.8905 0 0>, rotation=<rot -90 180 180>}]
 dyn_earring_01_r [{shapeType=Capsule, length=0.5, radius=0.5, shapeName=dyn_earring_01_capsule, center=<0.8905 0 0>, rotation=<rot -90 180 180>}]
 dyn_earring_02 [{shapeType=Capsule, length=0.5, radius=0.5, shapeName=dyn_earring_02_capsule, center=<0.8905 0 0>, rotation=<rot -90 180 180>}]
 dyn_earring_02_r [{shapeType=Capsule, length=0.5, radius=0.5, shapeName=dyn_earring_02_capsule, center=<0.8905 0 0>, rotation=<rot -90 180 180>}]
 dyn_earring_03 [{shapeType=Capsule, length=0.5, radius=0.5, shapeName=dyn_earring_03_capsule, center=<0.8905 0 0>, rotation=<rot -90 180 180>}]
 dyn_earring_03_r [{shapeType=Capsule, length=0.5, radius=0.5, shapeName=dyn_earring_03_capsule, center=<0.8905 0 0>, rotation=<rot -90 180 180>}]
 dyn_hairStrand_l_1 [{shapeType=Capsule, length=4, radius=2, shapeName=dyn_hairStrand_l_1_capsule, center=<8 -0.8 0>, rotation=<rot -90 0 0>}]
 dyn_hairStrand_l_2 [{shapeType=Capsule, length=4, radius=2, shapeName=dyn_hairStrand_l_2_capsule, center=<8 -0.8 0>, rotation=<rot -90 0 0>}]
 dyn_hair_ba_1 [{shapeType=Capsule, length=6, radius=4, shapeName=dyn_hair_ba_1_capsule, center=<5 -5.5 0>, rotation=<rot -90 0 0>}]
 dyn_hair_ba_2 [{shapeType=Capsule, length=6, radius=4, shapeName=dyn_hair_ba_2_capsule, center=<5 -3 0>, rotation=<rot -90 180 180>}]
 dyn_hair_l_1 [{shapeType=Capsule, length=6, radius=4, shapeName=dyn_hair_l_1_capsule, center=<5 0 3.8>, rotation=<rot -90 180 180>}]
 dyn_hair_l_2 [{shapeType=Capsule, length=6, radius=4, shapeName=dyn_hair_l_2_capsule, center=<5 0 3.8>, rotation=<rot -90 180 180>}]
 dyn_hair_r_1 [{shapeType=Capsule, length=6, radius=4, shapeName=dyn_hair_r_1_capsule, center=<5 0 -3.8>, rotation=<rot -90 0 0>}]
 dyn_hair_r_2 [{shapeType=Capsule, length=6, radius=4, shapeName=dyn_hair_r_2_capsule, center=<5 0 -3.8>, rotation=<rot -90 180 180>}]
 head [{shapeType=Capsule, length=13.8738, radius=7.6262, shapeName=head_capsule, center=<-5.5 1.2815 0>, rotation=<rot -90 0 0>}]
 spine_05 [{shapeType=Capsule, length=0.505, radius=0.505, shapeName=spine_05_capsule, rotation=<rot 0 -90 0>}]
}
constraints [
 {bone1Name=dyn_hairStrand_l_1, bone2Name=MSTR_hairStrand_l, swing1LimitDegrees=5, swing2LimitDegrees=20, twistLimitDegrees=0}
 {bone1Name=dyn_hairStrand_l_2, bone2Name=dyn_hairStrand_l_1, swing1LimitDegrees=5, swing2LimitDegrees=20, twistLimitDegrees=0}
 {bone1Name=dyn_hair_ba_1, bone2Name=MSTR_hair_ba, swing1LimitDegrees=12, swing2LimitDegrees=8, twistLimitDegrees=0}
 {bone1Name=dyn_hair_ba_2, bone2Name=dyn_hair_ba_1, swing1LimitDegrees=15, swing2LimitDegrees=15, twistLimitDegrees=0}
 {bone1Name=dyn_hair_l_1, bone2Name=MSTR_hair_l, swing1LimitDegrees=30, swing2LimitDegrees=15, twistLimitDegrees=0}
 {bone1Name=dyn_hair_l_2, bone2Name=dyn_hair_l_1, swing1LimitDegrees=30, swing2LimitDegrees=15, twistLimitDegrees=0}
 {bone1Name=dyn_hair_r_1, bone2Name=MSTR_hair_r, swing1LimitDegrees=15, swing2LimitDegrees=15, twistLimitDegrees=0}
 {bone1Name=dyn_hair_r_2, bone2Name=dyn_hair_r_1, swing1LimitDegrees=15, swing2LimitDegrees=20, twistLimitDegrees=0}
 {
  bone1Name=MSTR_hair_ba
  bone2Name=head
  swing1LimitDegrees=45
  swing1Motion=Free
  swing2LimitDegrees=45
  swing2Motion=Free
  twistLimitDegrees=45
  twistMotion=Free
 }
 {
  bone1Name=MSTR_hair_l
  bone2Name=head
  swing1LimitDegrees=45
  swing1Motion=Free
  swing2LimitDegrees=45
  swing2Motion=Free
  twistLimitDegrees=45
  twistMotion=Free
 }
 {
  bone1Name=MSTR_hair_r
  bone2Name=head
  swing1LimitDegrees=45
  swing1Motion=Free
  swing2LimitDegrees=45
  swing2Motion=Free
  twistLimitDegrees=45
  twistMotion=Free
 }
 {
  bone1Name=MSTR_hairStrand_l
  bone2Name=head
  swing1LimitDegrees=45
  swing1Motion=Free
  swing2LimitDegrees=45
  swing2Motion=Free
  twistLimitDegrees=45
  twistMotion=Free
 }
 {
  bone1Name=head
  bone2Name=spine_05
  swing1LimitDegrees=45
  swing1Motion=Free
  swing2LimitDegrees=45
  swing2Motion=Free
  twistLimitDegrees=45
  twistMotion=Free
 }
 {
  bone1Name=MSTR_earring
  bone2Name=head
  swing1LimitDegrees=45
  swing1Motion=Free
  swing2LimitDegrees=45
  swing2Motion=Free
  twistLimitDegrees=45
  twistMotion=Free
 }
 {bone1Name=dyn_earring_01, bone2Name=MSTR_earring, swing1LimitDegrees=15, swing2LimitDegrees=15, twistLimitDegrees=0}
 {bone1Name=dyn_earring_02, bone2Name=dyn_earring_01, swing1LimitDegrees=15, swing2LimitDegrees=15, twistLimitDegrees=0}
 {bone1Name=dyn_earring_03, bone2Name=dyn_earring_02, swing1LimitDegrees=15, swing2LimitDegrees=15, twistLimitDegrees=0}
 {
  bone1Name=MSTR_earring_r
  bone2Name=head
  swing1LimitDegrees=45
  swing1Motion=Free
  swing2LimitDegrees=45
  swing2Motion=Free
  twistLimitDegrees=45
  twistMotion=Free
 }
 {bone1Name=dyn_earring_01_r, bone2Name=MSTR_earring_r, swing1LimitDegrees=15, swing2LimitDegrees=15, twistLimitDegrees=0}
 {bone1Name=dyn_earring_02_r, bone2Name=dyn_earring_01_r, swing1LimitDegrees=15, swing2LimitDegrees=15, twistLimitDegrees=0}
 {bone1Name=dyn_earring_03_r, bone2Name=dyn_earring_02_r, swing1LimitDegrees=15, swing2LimitDegrees=15, twistLimitDegrees=0}
]
